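# Base editor utility widget for list item entries in the Sound Class Viewer audio tool, with no widgets of its own.
# It stores the list's item object on assignment and keeps a reference to the parent viewer through OnPostInitialize.
# Mouse clicks and double-clicks funnel into one OnClicked hook with a double-click flag.
# OnClicked and OnSelectedChanged are empty stubs, apparently for child item widgets to override.
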WidgetBlueprint /Game/Sounds/Fort_Audio_Blueprints/Utilities/EditorUtilityWidgets/SoundClassViewer/EUW_SoundClassViewer_ItemParent
graphs [
 EventGraph {nodes [
   n1 = Event OnListItemObjectSet()
   > Set EntryObject(EntryObject:Object Reference=@n1.ListItemObject) -> (Output_Get:Object Reference)
  ]}
 OnClicked {nodes [OnClicked!() -> (then, MouseEvent:"Pointer Event Structure (by ref)", bDoubleClick:bool=~) {is=title K2Node_FunctionEntry}]}
 OnMouseClicked {nodes [
   n1 = OnMouseClicked!() -> (then, MyGeometry:Geometry Structure, MouseEvent:"Pointer Event Structure (by ref)") {is=title K2Node_FunctionEntry}
   > OnClicked!(execute, self, MouseEvent:"Pointer Event Structure (by ref)"=@n1.MouseEvent, bDoubleClick:bool=false) -> (then) {is=title K2Node_CallFunction}
   > Return Node!(execute, ReturnValue:Event Reply Structure=Make Event Reply!() -> (EventReply:Event Reply Structure) {is=title K2Node_MakeStruct})
  ]}
 OnMouseDoubleClick {nodes [
   n1 = OnMouseDoubleClick!() -> (then, MyGeometry:Geometry Structure, MouseEvent:"Pointer Event Structure (by ref)") {is=title K2Node_FunctionEntry}
   > OnClicked(MouseEvent=@n1.MouseEvent, bDoubleClick=true)
   > Return Node(ReturnValue=Make Event Reply())
  ]}
 OnPostInitialize {nodes [
   n1 = OnPostInitialize!() -> (then, Parent Widget:EUW Sound Class Viewer Object Reference) {is=title K2Node_FunctionEntry}
   > Set Parent(Parent:EUW Sound Class Viewer Object Reference=@n1) -> (Output_Get:EUW Sound Class Viewer Object Reference)
  ]}
 OnSelectedChanged {nodes [OnSelectedChanged!() -> (then, bSelected:bool=~) {is=title K2Node_FunctionEntry}]}
]
variables [
 EntryObject {type=Object, default=null}
 Parent {type=EUW_SoundClassViewer_C, default=null}
]
widgets {info {inheritedWidgetCount=0, parentClass=&EditorUtilityWidget, rootWidgetClass=null, widgetCount=0}, widgets []}
