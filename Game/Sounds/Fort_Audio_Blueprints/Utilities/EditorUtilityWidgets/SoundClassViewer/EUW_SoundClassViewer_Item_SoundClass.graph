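# Row widget for one sound class in the Sound Class Viewer editor utility, showing its display name beside a plus/minus collapse checkbox.
# On post-initialize it recursively walks ParentClass up to the viewer's Root to find its depth, then indents the highlight border by that depth.
# Clicks and collapse toggles are forwarded to the parent viewer.
# Selection shows a faint highlight by changing the border's brush alpha.
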WidgetBlueprint /Game/Sounds/Fort_Audio_Blueprints/Utilities/EditorUtilityWidgets/SoundClassViewer/EUW_SoundClassViewer_Item_SoundClass
graphs [
 EventGraph {nodes [
   n1 = "On Check State Changed (CollapseCheckbox)"!() -> (OutputDelegate, then, bIsChecked:bool) {is=title K2Node_ComponentBoundEvent}
   > EventOnClassCollapsed!(
    execute
    self:EUW Sound Class Viewer Object Reference=Get Parent() -> (Parent:EUW Sound Class Viewer Object Reference)
    Widget:EUW Sound Class Viewer Item Sound Class Object Reference=Self-Reference()
    Is Collapsed:bool=NOT Boolean(A=@n1.bIsChecked)
   ) -> (then)
   n2 = Event OnListItemObjectSet()
   > Parent: OnListItemObjectSet!(execute, ListItemObject:Object Reference=@n2.ListItemObject) -> (then) {is=title K2Node_CallParentFunction}
   > "SetText (Text)"(self=Get Text!() -> (Text:Text Block Object Reference), InText="To Text (String)"(InString=GetDisplayName(Object=@n2.ListItemObject)))
   > SetPadding!(
    execute
    self:Border Object Reference=Get HighlightBorder() -> (HighlightBorder:Border Object Reference)
    InPadding_Left:float=0
    InPadding_Top:float=0
    InPadding_Right:float=0
    InPadding_Bottom:float=0
   ) -> (then)
  ]}
 GetChildDepth {nodes [
   GetChildDepth() {is=title K2Node_FunctionEntry}
   > n1 = RecursiveGetParent!(execute, self, Current Item:Object Reference=Get EntryObject() -> (EntryObject:Object Reference), Current Depth:int=0) -> (then, Result Depth:int=0)
   > Return Node!(execute, Depth:int=@n1)
  ]}
 OnClicked {nodes [
   n1 = OnClicked!() -> (then, MouseEvent:"Pointer Event Structure (by ref)", bDoubleClick:bool) {is=title K2Node_FunctionEntry}
   > Parent: OnClicked!(execute, MouseEvent:"Pointer Event Structure (by ref)"=@n1, bDoubleClick:bool=@n1.bDoubleClick) -> (then) {is=title K2Node_CallParentFunction}
   > EventOnClassClicked!(
    execute
    self:EUW Sound Class Viewer Object Reference=Get Parent()
    Widget:EUW Sound Class Viewer Item Sound Class Object Reference=Self-Reference()
    MouseEvent:Pointer Event Structure=@n1
    bDoubleClick:bool=@n1.bDoubleClick
   ) -> (then)
   > Return Node!(execute)
  ]}
 OnPostInitialize {nodes [
   n1 = OnPostInitialize!() -> (then, Parent Widget:EUW Sound Class Viewer Object Reference) {is=title K2Node_FunctionEntry}
   > Parent: OnPostInitialize!(execute, Parent Widget:EUW Sound Class Viewer Object Reference=@n1) -> (then) {is=title K2Node_CallParentFunction}
   > n2 = GetChildDepth!(execute, self) -> (then, Depth:int=0) {is=title K2Node_CallFunction}
   > SetPadding(
    self=Get HighlightBorder()
    InPadding_Left=float * float(A="To Float (Integer)"(InInt=@n2), B=25) {is=title K2Node_PromotableOperator}
    InPadding_Top=0
    InPadding_Right=0
    InPadding_Bottom=0
   )
  ]}
 OnSelectedChanged {nodes [
   n1 = OnSelectedChanged!() -> (then, bSelected:bool) {is=title K2Node_FunctionEntry}
   > Parent: OnSelectedChanged!(execute, bSelected:bool=@n1) -> (then) {is=title K2Node_CallParentFunction}
   n2 = Get HighlightBorder()
   n3 = Get!(self:Border Object Reference=@n2) -> (ContentColorAndOpacity_R:float=~, ContentColorAndOpacity_G:float=~, ContentColorAndOpacity_B:float=~, ContentColorAndOpacity_A:float=~)
   > SetBrushColor!(
    execute
    self=@n2
    InBrushColor_R:float=@n3
    InBrushColor_G:float=@n3.ContentColorAndOpacity_G
    InBrushColor_B:float=@n3.ContentColorAndOpacity_B
    InBrushColor_A:float=SelectFloat(A=0.1, B=0, bPickA=@n1)
   ) -> (then)
  ]}
 RecursiveGetParent {nodes [
   RecursiveGetParent!() -> (then, Current Item:Object Reference, Current Depth:int) {is=title K2Node_FunctionEntry}
   > n1 = Get!(execute:Exec, self:EUW Sound Class Viewer Object Reference=Get Parent()) -> (then:Exec, else:Exec, Root:Sound Class Object Reference)
   > n2 = Cast To SoundClass!(execute, Object=Get Current Item() -> (Current Item:Object Reference)) -> (then, CastFailed, AsSound Class:Sound Class Object Reference) {is=title K2Node_DynamicCast}
   > n3 = Branch(Condition="Equal (Object)"(A=@n1, B=@n2))
   > Return Node!(execute, Result Depth:int=Get Current Depth() -> (Current Depth:int=0))
   n4 = RecursiveGetParent(
    execute=@n3.else
    Current Item=Get ParentClass!(self:Sound Class Object Reference=@n2) -> (ParentClass:Sound Class Object Reference)
    Current Depth=int + int(A=Get Current Depth() -> (Current Depth=0))
   ) -> (Result Depth=0) {is=title K2Node_CallFunction}
   > Return Node(Result Depth=@n4)
   Return Node(execute=@n1.else, Result Depth=Get Current Depth() -> (Current Depth=0))
  ]}
]
variables [RootSoundClass {type=SoundClass, default=null, class=~, graph_type=~, children=~, properties=~}]
widgets {
 info {
  inheritedWidgetCount=0
  parentClass=&/EUW_SoundClassViewer_ItemParent.EUW_SoundClassViewer_ItemParent_C
  rootWidgetClass=&CanvasPanel
  widgetCount=6
 }
 widgets [
  CanvasPanel_178 {widgetClassPath=&CanvasPanel, widget=CanvasPanel_178}
  HighlightBorder {
   widgetClassPath=&Border
   bIsVariable=true
   parent=CanvasPanel_178
   widget=HighlightBorder
   properties {
    background {
     bIsDynamicallyLoaded=false
     drawAs=Box
     imageType=NoImage
     mirroring=NoMirror
     resourceName=None
     resourceObject=null
     tiling=NoTile
     imageSize=<32 32>
     margin=<ltrb 0 0 0 0>
     outlineSettings {
      bUseBrushTransparency=false
      roundingType=HalfHeightRadius
      width=0
      color {colorUseRule=UseColor_Specified, specifiedColor=<rgba 0 0 0 0>}
      cornerRadii=<0 0 0 0>
     }
     tintColor {colorUseRule=UseColor_Specified, specifiedColor=<rgba 1 1 1 1>}
     uVRegion {bIsValid=false, max=<0 0>, min=<0 0>}
    }
    brushColor=<rgba 1 1 1 0>
    padding=<ltrb 5 5 5 5>
   }
   slot {
    class=CanvasPanelSlot
    properties {bAutoSize=true, layoutData {alignment=<0 0>, anchors {maximum=<1 1>, minimum=<0 0>}, offsets=<ltrb 0 0 0 0>}}
   }
  }
  VerticalBox_142 {
   widgetClassPath=&VerticalBox
   parent=HighlightBorder
   widget=VerticalBox_142
   slot {class=BorderSlot, properties {padding=<ltrb 5 5 5 5>}}
  }
  HorizontalBox_82 {widgetClassPath=&HorizontalBox, parent=VerticalBox_142, widget=HorizontalBox_82, slot {class=VerticalBoxSlot}}
  CollapseCheckbox {
   widgetClassPath=/Script/UMG.CheckBox
   bIsVariable=true
   parent=HorizontalBox_82
   widget=CollapseCheckbox
   properties {
    checkedState=Checked
    widgetStyle {
     checkBoxType=CheckBox
     backgroundHoveredImage {
      bIsDynamicallyLoaded=false
      drawAs=NoDrawType
      imageType=NoImage
      mirroring=NoMirror
      resourceName=None
      resourceObject=null
      tiling=NoTile
      imageSize=<0 0>
      margin=<ltrb 0 0 0 0>
      outlineSettings {
       bUseBrushTransparency=false
       roundingType=HalfHeightRadius
       width=0
       color {colorUseRule=UseColor_Specified, specifiedColor=<rgba 0 0 0 0>}
       cornerRadii=<0 0 0 0>
      }
      tintColor {colorUseRule=UseColor_Specified, specifiedColor=<rgba 1 1 1 1>}
      uVRegion {bIsValid=false, max=<0 0>, min=<0 0>}
     }
     backgroundImage {
      bIsDynamicallyLoaded=false
      drawAs=NoDrawType
      imageType=NoImage
      mirroring=NoMirror
      resourceName=None
      resourceObject=null
      tiling=NoTile
      imageSize=<0 0>
      margin=<ltrb 0 0 0 0>
      outlineSettings {
       bUseBrushTransparency=false
       roundingType=HalfHeightRadius
       width=0
       color {colorUseRule=UseColor_Specified, specifiedColor=<rgba 0 0 0 0>}
       cornerRadii=<0 0 0 0>
      }
      tintColor {colorUseRule=UseColor_Specified, specifiedColor=<rgba 1 1 1 1>}
      uVRegion {bIsValid=false, max=<0 0>, min=<0 0>}
     }
     backgroundPressedImage {
      bIsDynamicallyLoaded=false
      drawAs=NoDrawType
      imageType=NoImage
      mirroring=NoMirror
      resourceName=None
      resourceObject=null
      tiling=NoTile
      imageSize=<0 0>
      margin=<ltrb 0 0 0 0>
      outlineSettings {
       bUseBrushTransparency=false
       roundingType=HalfHeightRadius
       width=0
       color {colorUseRule=UseColor_Specified, specifiedColor=<rgba 0 0 0 0>}
       cornerRadii=<0 0 0 0>
      }
      tintColor {colorUseRule=UseColor_Specified, specifiedColor=<rgba 1 1 1 1>}
      uVRegion {bIsValid=false, max=<0 0>, min=<0 0>}
     }
     borderBackgroundColor {colorUseRule=UseColor_Specified, specifiedColor=<rgba 1 1 1 1>}
     checkedForeground {colorUseRule=UseColor_Foreground, specifiedColor=<rgba 1 0 1 1>}
     checkedHoveredForeground {colorUseRule=UseColor_Foreground, specifiedColor=<rgba 1 0 1 1>}
     checkedHoveredImage {
      bIsDynamicallyLoaded=false
      drawAs=Image
      imageType=FullColor
      mirroring=NoMirror
      resourceName=../../../Engine/Content/Slate/Common/CheckBox.png
      resourceObject=/FortUILibrary/Foundation/Buttons/Materials/Arrows/M_Button_Minus_Hovered
      tiling=NoTile
      imageSize=<16 16>
      margin=<ltrb 0 0 0 0>
      outlineSettings {
       bUseBrushTransparency=false
       roundingType=HalfHeightRadius
       width=0
       color {colorUseRule=UseColor_Specified, specifiedColor=<rgba 0 0 0 0>}
       cornerRadii=<0 0 0 0>
      }
      tintColor {colorUseRule=UseColor_Specified, specifiedColor=<rgba 1 1 1 1>}
      uVRegion {bIsValid=false, max=<0 0>, min=<0 0>}
     }
     checkedImage {
      bIsDynamicallyLoaded=false
      drawAs=Image
      imageType=FullColor
      mirroring=NoMirror
      resourceName=../../../Engine/Content/Slate/Common/CheckBox.png
      resourceObject=/FortUILibrary/Foundation/Buttons/Materials/Arrows/M_Button_Minus_Base
      tiling=NoTile
      imageSize=<16 16>
      margin=<ltrb 0 0 0 0>
      outlineSettings {
       bUseBrushTransparency=false
       roundingType=HalfHeightRadius
       width=0
       color {colorUseRule=UseColor_Specified, specifiedColor=<rgba 0 0 0 0>}
       cornerRadii=<0 0 0 0>
      }
      tintColor {colorUseRule=UseColor_Specified, specifiedColor=<rgba 1 1 1 1>}
      uVRegion {bIsValid=false, max=<0 0>, min=<0 0>}
     }
     checkedPressedForeground {colorUseRule=UseColor_Foreground, specifiedColor=<rgba 1 0 1 1>}
     checkedPressedImage {
      bIsDynamicallyLoaded=false
      drawAs=Image
      imageType=FullColor
      mirroring=NoMirror
      resourceName=../../../Engine/Content/Slate/Common/CheckBox.png
      resourceObject=/FortUILibrary/Foundation/Buttons/Materials/Arrows/M_Button_Minus_Pressed
      tiling=NoTile
      imageSize=<16 16>
      margin=<ltrb 0 0 0 0>
      outlineSettings {
       bUseBrushTransparency=false
       roundingType=HalfHeightRadius
       width=0
       color {colorUseRule=UseColor_Specified, specifiedColor=<rgba 0 0 0 0>}
       cornerRadii=<0 0 0 0>
      }
      tintColor {colorUseRule=UseColor_Specified, specifiedColor=<rgba 1 1 1 1>}
      uVRegion {bIsValid=false, max=<0 0>, min=<0 0>}
     }
     checkedSlateSound {resourceObject=null}
     foregroundColor {colorUseRule=UseColor_Foreground, specifiedColor=<rgba 1 0 1 1>}
     hoveredForeground {colorUseRule=UseColor_Foreground, specifiedColor=<rgba 1 0 1 1>}
     hoveredSlateSound {resourceObject=null}
     padding=<ltrb 2 0 0 0>
     pressedForeground {colorUseRule=UseColor_Foreground, specifiedColor=<rgba 1 0 1 1>}
     uncheckedHoveredImage {
      bIsDynamicallyLoaded=false
      drawAs=Image
      imageType=FullColor
      mirroring=NoMirror
      resourceName=../../../Engine/Content/Slate/Common/CheckBox.png
      resourceObject=/FortUILibrary/Foundation/Buttons/Materials/Arrows/M_Button_Plus_Hovered
      tiling=NoTile
      imageSize=<16 16>
      margin=<ltrb 0 0 0 0>
      outlineSettings {
       bUseBrushTransparency=false
       roundingType=HalfHeightRadius
       width=0
       color {colorUseRule=UseColor_Specified, specifiedColor=<rgba 0 0 0 0>}
       cornerRadii=<0 0 0 0>
      }
      tintColor {colorUseRule=UseColor_Specified, specifiedColor=<rgba 1 1 1 1>}
      uVRegion {bIsValid=false, max=<0 0>, min=<0 0>}
     }
     uncheckedImage {
      bIsDynamicallyLoaded=false
      drawAs=Image
      imageType=FullColor
      mirroring=NoMirror
      resourceName=../../../Engine/Content/Slate/Common/CheckBox.png
      resourceObject=/FortUILibrary/Foundation/Buttons/Materials/Arrows/M_Button_Plus_Base
      tiling=NoTile
      imageSize=<16 16>
      margin=<ltrb 0 0 0 0>
      outlineSettings {
       bUseBrushTransparency=false
       roundingType=HalfHeightRadius
       width=0
       color {colorUseRule=UseColor_Specified, specifiedColor=<rgba 0 0 0 0>}
       cornerRadii=<0 0 0 0>
      }
      tintColor {colorUseRule=UseColor_Specified, specifiedColor=<rgba 1 1 1 1>}
      uVRegion {bIsValid=false, max=<0 0>, min=<0 0>}
     }
     uncheckedPressedImage {
      bIsDynamicallyLoaded=false
      drawAs=Image
      imageType=FullColor
      mirroring=NoMirror
      resourceName=../../../Engine/Content/Slate/Common/CheckBox.png
      resourceObject=/FortUILibrary/Foundation/Buttons/Materials/Arrows/M_Button_Plus_Pressed
      tiling=NoTile
      imageSize=<16 16>
      margin=<ltrb 0 0 0 0>
      outlineSettings {
       bUseBrushTransparency=false
       roundingType=HalfHeightRadius
       width=0
       color {colorUseRule=UseColor_Specified, specifiedColor=<rgba 0 0 0 0>}
       cornerRadii=<0 0 0 0>
      }
      tintColor {colorUseRule=UseColor_Specified, specifiedColor=<rgba 1 1 1 1>}
      uVRegion {bIsValid=false, max=<0 0>, min=<0 0>}
     }
     uncheckedSlateSound {resourceObject=null}
     undeterminedForeground {colorUseRule=UseColor_Foreground, specifiedColor=<rgba 1 0 1 1>}
     undeterminedHoveredImage {
      bIsDynamicallyLoaded=false
      drawAs=Image
      imageType=FullColor
      mirroring=NoMirror
      resourceName=../../../Engine/Content/Slate/Common/CheckBox_Undetermined_Hovered.png
      resourceObject=null
      tiling=NoTile
      imageSize=<16 16>
      margin=<ltrb 0 0 0 0>
      outlineSettings {
       bUseBrushTransparency=false
       roundingType=HalfHeightRadius
       width=0
       color {colorUseRule=UseColor_Specified, specifiedColor=<rgba 0 0 0 0>}
       cornerRadii=<0 0 0 0>
      }
      tintColor {colorUseRule=UseColor_Specified, specifiedColor=<rgba 1 1 1 1>}
      uVRegion {bIsValid=false, max=<0 0>, min=<0 0>}
     }
     undeterminedImage {
      bIsDynamicallyLoaded=false
      drawAs=Image
      imageType=FullColor
      mirroring=NoMirror
      resourceName=../../../Engine/Content/Slate/Common/CheckBox_Undetermined.png
      resourceObject=null
      tiling=NoTile
      imageSize=<16 16>
      margin=<ltrb 0 0 0 0>
      outlineSettings {
       bUseBrushTransparency=false
       roundingType=HalfHeightRadius
       width=0
       color {colorUseRule=UseColor_Specified, specifiedColor=<rgba 0 0 0 0>}
       cornerRadii=<0 0 0 0>
      }
      tintColor {colorUseRule=UseColor_Specified, specifiedColor=<rgba 1 1 1 1>}
      uVRegion {bIsValid=false, max=<0 0>, min=<0 0>}
     }
     undeterminedPressedImage {
      bIsDynamicallyLoaded=false
      drawAs=Image
      imageType=FullColor
      mirroring=NoMirror
      resourceName=../../../Engine/Content/Slate/Common/CheckBox_Undetermined_Hovered.png
      resourceObject=null
      tiling=NoTile
      imageSize=<16 16>
      margin=<ltrb 0 0 0 0>
      outlineSettings {
       bUseBrushTransparency=false
       roundingType=HalfHeightRadius
       width=0
       color {colorUseRule=UseColor_Specified, specifiedColor=<rgba 0 0 0 0>}
       cornerRadii=<0 0 0 0>
      }
      tintColor {colorUseRule=UseColor_Specified, specifiedColor=<rgba 0.5 0.5 0.5 1>}
      uVRegion {bIsValid=false, max=<0 0>, min=<0 0>}
     }
    }
   }
   slot {class=HorizontalBoxSlot, properties {padding=<ltrb 0 0 5 0>}}
  }
  Text {
   widgetClassPath=&UEFN_TextBlock
   bIsVariable=true
   parent=HorizontalBox_82
   widget=Text
   properties {
    text=Actor
    font {
     bForceMonospaced=false
     bMaterialIsStencil=false
     fontMaterial=null
     fontObject=&NotoSans
     letterSpacing=5
     monospacedWidth=1
     size=10
     skewAmount=0
     typefaceFontName=NotoSans-Regular
     outlineSettings {
      bApplyOutlineToDropShadows=false
      bMiteredCorners=false
      bSeparateFillAlpha=false
      outlineMaterial=null
      outlineSize=0
      outlineColor=<rgba 0 0 0 1>
     }
    }
   }
   slot {class=HorizontalBoxSlot, properties {padding=<ltrb 0 0 200 0>}}
  }
 ]
}
