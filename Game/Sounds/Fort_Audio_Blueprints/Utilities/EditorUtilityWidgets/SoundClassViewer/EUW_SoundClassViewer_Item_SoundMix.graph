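# Editor-only list row in the Sound Class Viewer tool, showing a SoundMix's name and its volume adjuster for the viewer's selected Sound Class.
# The GetAdjuster macro scans the mix's SoundClassEffects for that class and returns the entry by reference.
# Pressing Enter in the text box writes the parsed volume into that entry, then calls TransactObject on the mix.
# Clicks forward to the parent viewer, and selection tints the row border.
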WidgetBlueprint /Game/Sounds/Fort_Audio_Blueprints/Utilities/EditorUtilityWidgets/SoundClassViewer/EUW_SoundClassViewer_Item_SoundMix
graphs [
 EventGraph {nodes [
   n1 = "On Text Committed (VolumeModifierTextBox)"!() -> (OutputDelegate, then, Text:"Text (by ref)", CommitMethod:ETextCommit Enum=Default) {is=title K2Node_ComponentBoundEvent}
   > n2 = Switch on ETextCommit!(execute, Selection:ETextCommit Enum=@n1.CommitMethod) -> (Default:Exec, OnEnter:Exec, OnUserMovedFocus:Exec, OnCleared:Exec) {is=title K2Node_SwitchEnum}
   n3 = Get Adjuster!(Exec:Exec=@n2.OnEnter) -> (Success:Exec, Adjuster:"Sound Class Adjuster Structure (by ref)", Fail:Exec) {is=title K2Node_MacroInstance}
   > Set members in Sound Class Adjuster!(
    execute:Exec
    StructRef:"Sound Class Adjuster Structure (by ref)"=@n3
    VolumeAdjuster:float=String To Float(InString="To String (Text)"(InText=@n1.Text))
   ) -> (then:Exec, StructOut:"Sound Class Adjuster Structure (by ref)") {is=title K2Node_SetFieldsInStruct}
   > TransactObject(Object=Get EntryObject() -> (EntryObject:Object Reference))
   n4 = Event OnListItemObjectSet()
   > Parent: OnListItemObjectSet!(execute, ListItemObject:Object Reference=@n4.ListItemObject) -> (then) {is=title K2Node_CallParentFunction}
   > "SetText (Text)"(self=Get Text_1() -> (Text_1:Text Block Object Reference), InText="To Text (String)"(InString=GetDisplayName(Object=@n4.ListItemObject)))
  ]}
 GetAdjuster {nodes [
   Inputs!() -> (Exec:Exec)
   n1 = Get SoundClassEffects!(self:Sound Mix Object Reference=Cast To SoundMix!(Object=Get EntryObject()) -> (AsSound Mix:Sound Mix Object Reference, bSuccess:bool=~) {is=title K2Node_DynamicCast}) -> (SoundClassEffects:Array of Sound Class Adjuster Structures)
   > n2 = For Each Loop(Array:Array of Sound Class Adjuster Structures=@n1) -> (Array Element:Sound Class Adjuster Structure)
   n3 = "Get (a ref)"(Array:Array of Sound Class Adjuster Structures=@n1, Dimension 1=@n2.Array Index) -> (Output:"Sound Class Adjuster Structure (by ref)")
   Outputs!(Success:Exec=@n5.then, Adjuster:"Sound Class Adjuster Structure (by ref)"=@n3, Fail:Exec=[@n5.else, @n2.Completed])
   n4 = Get Parent() -> (Parent:EUW Sound Class Viewer Object Reference)
   > n5 = Branch(Condition="Equal (Object)"(
     A=Get Selected!(self:EUW Sound Class Viewer Object Reference=@n4) -> (Selected:Sound Class Object Reference)
     B=Break Sound Class Adjuster!(SoundClassAdjuster:"Sound Class Adjuster Structure (by ref)"=@n3) -> (SoundClassObject:Sound Class Object Reference) {is=title K2Node_BreakStruct}
    ))
  ]}
 GetChildDepth {nodes [
   GetChildDepth() {is=title K2Node_FunctionEntry}
   > n1 = RecursiveGetParent!(execute, self, Current Item:Object Reference=Get EntryObject(), Current Depth:int=0) -> (then, Result Depth:int=0)
   > Return Node!(execute, Depth:int=@n1)
  ]}
 OnClicked {nodes [
   n1 = OnClicked!() -> (then, MouseEvent:"Pointer Event Structure (by ref)", bDoubleClick:bool) {is=title K2Node_FunctionEntry}
   > Parent: OnClicked!(execute, MouseEvent:"Pointer Event Structure (by ref)"=@n1, bDoubleClick:bool=@n1.bDoubleClick) -> (then) {is=title K2Node_CallParentFunction}
   > EventOnMixClicked!(
    execute
    self:EUW Sound Class Viewer Object Reference=Get Parent()
    Widget:EUW Sound Class Viewer Item Sound Mix Object Reference=Self-Reference()
    MouseEvent:Pointer Event Structure=@n1
    bDoubleClick:bool=@n1.bDoubleClick
   ) -> (then)
   > Return Node!(execute)
  ]}
 OnPostInitialize {nodes [
   n1 = OnPostInitialize!() -> (then, Parent Widget:EUW Sound Class Viewer Object Reference) {is=title K2Node_FunctionEntry}
   > Parent: OnPostInitialize!(execute, Parent Widget:EUW Sound Class Viewer Object Reference=@n1) -> (then) {is=title K2Node_CallParentFunction}
   > n2 = Get Adjuster()
   > "SetText (Text Box)"!(
    execute
    self:Text Box Object Reference=Get VolumeModifierTextBox() -> (VolumeModifierTextBox:Text Box Object Reference)
    InText:Text=Format Text!(
     Format="LOCTEXT(\"{Volume}\")"
     Volume:float=Break Sound Class Adjuster!(SoundClassAdjuster=@n2) -> (
      SoundClassObject
      VolumeAdjuster:float=~
      PitchAdjuster:float=~
      LowPassFilterFrequency:float=~
      bApplyToChildren:bool=~
      VoiceCenterChannelVolumeAdjuster:float=~
     ).VolumeAdjuster
    ) -> (Result)
   ) -> (then)
  ]}
 OnSelectedChanged {nodes [
   n1 = OnSelectedChanged!() -> (then, bSelected:bool) {is=title K2Node_FunctionEntry}
   > Parent: OnSelectedChanged!(execute, bSelected:bool=@n1) -> (then) {is=title K2Node_CallParentFunction}
   n2 = Get Border_56() -> (Border_56:Border Object Reference)
   n3 = Get!(self:Border Object Reference=@n2) -> (ContentColorAndOpacity_R:float=~, ContentColorAndOpacity_G:float=~, ContentColorAndOpacity_B:float=~, ContentColorAndOpacity_A:float=~)
   > SetBrushColor!(
    execute
    self=@n2
    InBrushColor_R:float=@n3
    InBrushColor_G:float=@n3.ContentColorAndOpacity_G
    InBrushColor_B:float=@n3.ContentColorAndOpacity_B
    InBrushColor_A:float=SelectFloat(A=0.1, B=0, bPickA=@n1)
   ) -> (then)
  ]}
 RecursiveGetParent {nodes [
   RecursiveGetParent!() -> (then, Current Item:Object Reference, Current Depth:int) {is=title K2Node_FunctionEntry}
   > n1 = Get!(execute:Exec, self:EUW Sound Class Viewer Object Reference=Get Parent()) -> (then:Exec, else:Exec, Root:Sound Class Object Reference)
   > n2 = Cast To SoundClass!(execute, Object=Get Current Item() -> (Current Item:Object Reference)) -> (then, CastFailed, AsSound Class:Sound Class Object Reference) {is=title K2Node_DynamicCast}
   > n3 = Branch(Condition="Equal (Object)"(A=@n1, B=@n2))
   > Return Node!(execute, Result Depth:int=Get Current Depth() -> (Current Depth:int=0))
   n4 = RecursiveGetParent(
    execute=@n3.else
    Current Item=Get ParentClass!(self:Sound Class Object Reference=@n2) -> (ParentClass:Sound Class Object Reference)
    Current Depth=int + int(A=Get Current Depth() -> (Current Depth=0))
   ) -> (Result Depth=0) {is=title K2Node_CallFunction}
   > Return Node(Result Depth=@n4)
   Return Node(execute=@n1.else, Result Depth=Get Current Depth() -> (Current Depth=0))
  ]}
]
variables []
widgets {
 info {
  inheritedWidgetCount=0
  parentClass=&/EUW_SoundClassViewer_ItemParent.EUW_SoundClassViewer_ItemParent_C
  rootWidgetClass=&CanvasPanel
  widgetCount=7
 }
 widgets [
  CanvasPanel_178 {widgetClassPath=&CanvasPanel, widget=CanvasPanel_178}
  VerticalBox_142 {
   widgetClassPath=&VerticalBox
   parent=CanvasPanel_178
   widget=VerticalBox_142
   slot {
    class=CanvasPanelSlot
    properties {bAutoSize=true, layoutData {alignment=<0 0>, anchors {maximum=<1 1>, minimum=<0 0>}, offsets=<ltrb 0 0 0 0>}}
   }
  }
  HorizontalBox_82 {widgetClassPath=&HorizontalBox, parent=VerticalBox_142, widget=HorizontalBox_82, slot {class=VerticalBoxSlot}}
  Border_56 {
   widgetClassPath=&Border
   bIsVariable=true
   parent=HorizontalBox_82
   widget=Border_56
   properties {brushColor=<rgba 1 1 1 0>, padding=<ltrb 0 0 0 0>}
   slot {class=HorizontalBoxSlot, properties {size {sizeRule=Fill, value=1}}}
  }
  UniformGridPanel_63 {
   widgetClassPath=/Script/UMG.UniformGridPanel
   parent=Border_56
   widget=UniformGridPanel_63
   slot {class=BorderSlot, properties {padding=<ltrb 0 0 0 0>}}
  }
  Text_1 {
   widgetClassPath=&UEFN_TextBlock
   bIsVariable=true
   parent=UniformGridPanel_63
   widget=Text_1
   properties {
    clipping=ClipToBounds
    text=SoundMix Asset
    font {
     bForceMonospaced=false
     bMaterialIsStencil=false
     fontMaterial=null
     fontObject=&NotoSans
     letterSpacing=5
     monospacedWidth=1
     size=10
     skewAmount=0
     typefaceFontName=NotoSans-Regular
     outlineSettings {
      bApplyOutlineToDropShadows=false
      bMiteredCorners=false
      bSeparateFillAlpha=false
      outlineMaterial=null
      outlineSize=0
      outlineColor=<rgba 0 0 0 1>
     }
    }
   }
   slot {class=UniformGridSlot, properties {horizontalAlignment=HAlign_Fill, verticalAlignment=VAlign_Fill}}
  }
  VolumeModifierTextBox {
   widgetClassPath=&EditableTextBox
   bIsVariable=true
   parent=UniformGridPanel_63
   widget=VolumeModifierTextBox
   properties {
    minimumDesiredWidth=200
    widgetStyle {
     backgroundColor {colorUseRule=UseColor_Specified, specifiedColor=<rgba 1 1 1 1>}
     backgroundImageFocused {
      bIsDynamicallyLoaded=false
      drawAs=Box
      imageType=FullColor
      mirroring=NoMirror
      resourceName=../../../Engine/Content/Slate/Common/TextBox_Hovered.png
      resourceObject=null
      tiling=NoTile
      imageSize=<0 0>
      margin=<ltrb 0.25 0.25 0.25 0.25>
      outlineSettings {
       bUseBrushTransparency=false
       roundingType=HalfHeightRadius
       width=0
       color {colorUseRule=UseColor_Specified, specifiedColor=<rgba 0 0 0 0>}
       cornerRadii=<0 0 0 0>
      }
      tintColor {colorUseRule=UseColor_Specified, specifiedColor=<rgba 0 0 0 0.2>}
      uVRegion {bIsValid=false, max=<0 0>, min=<0 0>}
     }
     backgroundImageHovered {
      bIsDynamicallyLoaded=false
      drawAs=Box
      imageType=FullColor
      mirroring=NoMirror
      resourceName=../../../Engine/Content/Slate/Common/TextBox_Hovered.png
      resourceObject=null
      tiling=NoTile
      imageSize=<0 0>
      margin=<ltrb 0.25 0.25 0.25 0.25>
      outlineSettings {
       bUseBrushTransparency=false
       roundingType=HalfHeightRadius
       width=0
       color {colorUseRule=UseColor_Specified, specifiedColor=<rgba 0 0 0 0>}
       cornerRadii=<0 0 0 0>
      }
      tintColor {colorUseRule=UseColor_Specified, specifiedColor=<rgba 1 1 1 0.1>}
      uVRegion {bIsValid=false, max=<0 0>, min=<0 0>}
     }
     backgroundImageNormal {
      bIsDynamicallyLoaded=false
      drawAs=Box
      imageType=FullColor
      mirroring=NoMirror
      resourceName=../../../Engine/Content/Slate/Common/TextBox.png
      resourceObject=null
      tiling=NoTile
      imageSize=<0 0>
      margin=<ltrb 0.25 0.25 0.25 0.25>
      outlineSettings {
       bUseBrushTransparency=false
       roundingType=HalfHeightRadius
       width=0
       color {colorUseRule=UseColor_Specified, specifiedColor=<rgba 0 0 0 0>}
       cornerRadii=<0 0 0 0>
      }
      tintColor {colorUseRule=UseColor_Specified, specifiedColor=<rgba 0 0 0 0.2>}
      uVRegion {bIsValid=false, max=<0 0>, min=<0 0>}
     }
     backgroundImageReadOnly {
      bIsDynamicallyLoaded=false
      drawAs=Box
      imageType=FullColor
      mirroring=NoMirror
      resourceName=../../../Engine/Content/Slate/Common/TextBox_ReadOnly.png
      resourceObject=null
      tiling=NoTile
      imageSize=<0 0>
      margin=<ltrb 0.25 0.25 0.25 0.25>
      outlineSettings {
       bUseBrushTransparency=false
       roundingType=HalfHeightRadius
       width=0
       color {colorUseRule=UseColor_Specified, specifiedColor=<rgba 0 0 0 0>}
       cornerRadii=<0 0 0 0>
      }
      tintColor {colorUseRule=UseColor_Specified, specifiedColor=<rgba 1 1 1 1>}
      uVRegion {bIsValid=false, max=<0 0>, min=<0 0>}
     }
     focusedForegroundColor {colorUseRule=UseColor_Foreground, specifiedColor=<rgba 1 0 1 1>}
     foregroundColor {colorUseRule=UseColor_Foreground, specifiedColor=<rgba 1 0 1 1>}
     hScrollBarPadding=<ltrb 0 0 0 0>
     padding=<ltrb 4 2 4 2>
     readOnlyForegroundColor {colorUseRule=UseColor_Foreground, specifiedColor=<rgba 1 0 1 1>}
     scrollBarStyle {
      thickness=16
      draggedThumbImage {
       bIsDynamicallyLoaded=false
       drawAs=Box
       imageType=FullColor
       mirroring=NoMirror
       resourceName=../../../Engine/Content/Slate/Common/Scrollbar_Thumb.png
       resourceObject=null
       tiling=NoTile
       imageSize=<0 0>
       margin=<ltrb 0.25 0.25 0.25 0.25>
       outlineSettings {
        bUseBrushTransparency=false
        roundingType=HalfHeightRadius
        width=0
        color {colorUseRule=UseColor_Specified, specifiedColor=<rgba 0 0 0 0>}
        cornerRadii=<0 0 0 0>
       }
       tintColor {colorUseRule=UseColor_Specified, specifiedColor=<rgba 1 1 1 1>}
       uVRegion {bIsValid=false, max=<0 0>, min=<0 0>}
      }
      horizontalBackgroundImage {
       bIsDynamicallyLoaded=false
       drawAs=NoDrawType
       imageType=NoImage
       mirroring=NoMirror
       resourceName=None
       resourceObject=null
       tiling=NoTile
       imageSize=<0 0>
       margin=<ltrb 0 0 0 0>
       outlineSettings {
        bUseBrushTransparency=false
        roundingType=HalfHeightRadius
        width=0
        color {colorUseRule=UseColor_Specified, specifiedColor=<rgba 0 0 0 0>}
        cornerRadii=<0 0 0 0>
       }
       tintColor {colorUseRule=UseColor_Specified, specifiedColor=<rgba 1 1 1 1>}
       uVRegion {bIsValid=false, max=<0 0>, min=<0 0>}
      }
      horizontalBottomSlotImage {
       bIsDynamicallyLoaded=false
       drawAs=Image
       imageType=FullColor
       mirroring=NoMirror
       resourceName=../../../Engine/Content/Slate/Common/Scrollbar_Background_Horizontal.png
       resourceObject=null
       tiling=NoTile
       imageSize=<8 8>
       margin=<ltrb 0 0 0 0>
       outlineSettings {
        bUseBrushTransparency=false
        roundingType=HalfHeightRadius
        width=0
        color {colorUseRule=UseColor_Specified, specifiedColor=<rgba 0 0 0 0>}
        cornerRadii=<0 0 0 0>
       }
       tintColor {colorUseRule=UseColor_Specified, specifiedColor=<rgba 1 1 1 1>}
       uVRegion {bIsValid=false, max=<0 0>, min=<0 0>}
      }
      horizontalTopSlotImage {
       bIsDynamicallyLoaded=false
       drawAs=Image
       imageType=FullColor
       mirroring=NoMirror
       resourceName=../../../Engine/Content/Slate/Common/Scrollbar_Background_Horizontal.png
       resourceObject=null
       tiling=NoTile
       imageSize=<8 8>
       margin=<ltrb 0 0 0 0>
       outlineSettings {
        bUseBrushTransparency=false
        roundingType=HalfHeightRadius
        width=0
        color {colorUseRule=UseColor_Specified, specifiedColor=<rgba 0 0 0 0>}
        cornerRadii=<0 0 0 0>
       }
       tintColor {colorUseRule=UseColor_Specified, specifiedColor=<rgba 1 1 1 1>}
       uVRegion {bIsValid=false, max=<0 0>, min=<0 0>}
      }
      hoveredThumbImage {
       bIsDynamicallyLoaded=false
       drawAs=Box
       imageType=FullColor
       mirroring=NoMirror
       resourceName=../../../Engine/Content/Slate/Common/Scrollbar_Thumb.png
       resourceObject=null
       tiling=NoTile
       imageSize=<0 0>
       margin=<ltrb 0.25 0.25 0.25 0.25>
       outlineSettings {
        bUseBrushTransparency=false
        roundingType=HalfHeightRadius
        width=0
        color {colorUseRule=UseColor_Specified, specifiedColor=<rgba 0 0 0 0>}
        cornerRadii=<0 0 0 0>
       }
       tintColor {colorUseRule=UseColor_Specified, specifiedColor=<rgba 1 1 1 1>}
       uVRegion {bIsValid=false, max=<0 0>, min=<0 0>}
      }
      normalThumbImage {
       bIsDynamicallyLoaded=false
       drawAs=Box
       imageType=FullColor
       mirroring=NoMirror
       resourceName=../../../Engine/Content/Slate/Common/Scrollbar_Thumb.png
       resourceObject=null
       tiling=NoTile
       imageSize=<0 0>
       margin=<ltrb 0.25 0.25 0.25 0.25>
       outlineSettings {
        bUseBrushTransparency=false
        roundingType=HalfHeightRadius
        width=0
        color {colorUseRule=UseColor_Specified, specifiedColor=<rgba 0 0 0 0>}
        cornerRadii=<0 0 0 0>
       }
       tintColor {colorUseRule=UseColor_Specified, specifiedColor=<rgba 1 1 1 1>}
       uVRegion {bIsValid=false, max=<0 0>, min=<0 0>}
      }
      verticalBackgroundImage {
       bIsDynamicallyLoaded=false
       drawAs=NoDrawType
       imageType=NoImage
       mirroring=NoMirror
       resourceName=None
       resourceObject=null
       tiling=NoTile
       imageSize=<0 0>
       margin=<ltrb 0 0 0 0>
       outlineSettings {
        bUseBrushTransparency=false
        roundingType=HalfHeightRadius
        width=0
        color {colorUseRule=UseColor_Specified, specifiedColor=<rgba 0 0 0 0>}
        cornerRadii=<0 0 0 0>
       }
       tintColor {colorUseRule=UseColor_Specified, specifiedColor=<rgba 1 1 1 1>}
       uVRegion {bIsValid=false, max=<0 0>, min=<0 0>}
      }
      verticalBottomSlotImage {
       bIsDynamicallyLoaded=false
       drawAs=Image
       imageType=FullColor
       mirroring=NoMirror
       resourceName=../../../Engine/Content/Slate/Common/Scrollbar_Background_Vertical.png
       resourceObject=null
       tiling=NoTile
       imageSize=<8 8>
       margin=<ltrb 0 0 0 0>
       outlineSettings {
        bUseBrushTransparency=false
        roundingType=HalfHeightRadius
        width=0
        color {colorUseRule=UseColor_Specified, specifiedColor=<rgba 0 0 0 0>}
        cornerRadii=<0 0 0 0>
       }
       tintColor {colorUseRule=UseColor_Specified, specifiedColor=<rgba 1 1 1 1>}
       uVRegion {bIsValid=false, max=<0 0>, min=<0 0>}
      }
      verticalTopSlotImage {
       bIsDynamicallyLoaded=false
       drawAs=Image
       imageType=FullColor
       mirroring=NoMirror
       resourceName=../../../Engine/Content/Slate/Common/Scrollbar_Background_Vertical.png
       resourceObject=null
       tiling=NoTile
       imageSize=<8 8>
       margin=<ltrb 0 0 0 0>
       outlineSettings {
        bUseBrushTransparency=false
        roundingType=HalfHeightRadius
        width=0
        color {colorUseRule=UseColor_Specified, specifiedColor=<rgba 0 0 0 0>}
        cornerRadii=<0 0 0 0>
       }
       tintColor {colorUseRule=UseColor_Specified, specifiedColor=<rgba 1 1 1 1>}
       uVRegion {bIsValid=false, max=<0 0>, min=<0 0>}
      }
     }
     textStyle {
      overflowPolicy=Clip
      transformPolicy=None
      colorAndOpacity {colorUseRule=UseColor_Foreground, specifiedColor=<rgba 1 0 1 1>}
      font {
       bForceMonospaced=false
       bMaterialIsStencil=false
       fontMaterial=null
       fontObject=null
       letterSpacing=0
       monospacedWidth=1
       size=9
       skewAmount=0
       typefaceFontName=Regular
       outlineSettings {
        bApplyOutlineToDropShadows=false
        bMiteredCorners=false
        bSeparateFillAlpha=false
        outlineMaterial=null
        outlineSize=0
        outlineColor=<rgba 0 0 0 1>
       }
      }
      highlightColor {colorUseRule=UseColor_Specified, specifiedColor=<rgba 0.02 0.3 0 1>}
      highlightShape {
       bIsDynamicallyLoaded=false
       drawAs=Box
       imageType=FullColor
       mirroring=NoMirror
       resourceName=../../../Engine/Content/Slate/Common/TextBlockHighlightShape.png
       resourceObject=null
       tiling=NoTile
       imageSize=<0 0>
       margin=<ltrb 0.375 0.375 0.375 0.375>
       outlineSettings {
        bUseBrushTransparency=false
        roundingType=HalfHeightRadius
        width=0
        color {colorUseRule=UseColor_Specified, specifiedColor=<rgba 0 0 0 0>}
        cornerRadii=<0 0 0 0>
       }
       tintColor {colorUseRule=UseColor_Specified, specifiedColor=<rgba 1 1 1 1>}
       uVRegion {bIsValid=false, max=<0 0>, min=<0 0>}
      }
      selectedBackgroundColor {colorUseRule=UseColor_Foreground, specifiedColor=<rgba 1 0 1 1>}
      shadowColorAndOpacity=<rgba 0 0 0 1>
      shadowOffset=<0 0>
      strikeBrush {
       bIsDynamicallyLoaded=false
       drawAs=Image
       imageType=NoImage
       mirroring=NoMirror
       resourceName=None
       resourceObject=null
       tiling=NoTile
       imageSize=<32 32>
       margin=<ltrb 0 0 0 0>
       outlineSettings {
        bUseBrushTransparency=false
        roundingType=HalfHeightRadius
        width=0
        color {colorUseRule=UseColor_Specified, specifiedColor=<rgba 0 0 0 0>}
        cornerRadii=<0 0 0 0>
       }
       tintColor {colorUseRule=UseColor_Specified, specifiedColor=<rgba 1 1 1 1>}
       uVRegion {bIsValid=false, max=<0 0>, min=<0 0>}
      }
      underlineBrush {
       bIsDynamicallyLoaded=false
       drawAs=Image
       imageType=NoImage
       mirroring=NoMirror
       resourceName=None
       resourceObject=null
       tiling=NoTile
       imageSize=<32 32>
       margin=<ltrb 0 0 0 0>
       outlineSettings {
        bUseBrushTransparency=false
        roundingType=HalfHeightRadius
        width=0
        color {colorUseRule=UseColor_Specified, specifiedColor=<rgba 0 0 0 0>}
        cornerRadii=<0 0 0 0>
       }
       tintColor {colorUseRule=UseColor_Specified, specifiedColor=<rgba 1 1 1 1>}
       uVRegion {bIsValid=false, max=<0 0>, min=<0 0>}
      }
     }
     vScrollBarPadding=<ltrb 0 0 0 0>
    }
   }
   slot {class=UniformGridSlot, properties {column=1}}
  }
 ]
}
